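# Post-process material for the Galileo prototype that adds screen-space camera shake and a wind-like radial smear, driven by MPC_Galileo parameters.
# Time-scrolled HighContrastNoise offsets screen UVs in XY and scales them about center for Z shake.
# CameraWind, panning noise and two DitherTemporalAA calls push two further samples radially from screen center, faking a blur.
# PostProcessBlend mixes the shaken image with the average of those two samples.
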Material /Game/Athena/Prototype/Blueprints/Galileo/PostProcess/M_PP_CameraShake
graph {
 properties {materialDomain=MD_PostProcess, pixelDepthOffsetMode=PDOM_Legacy, usedShadingModels=Default Lit}
 nodes [
  n1 = Time()
  n2 = Divide(
   A=Multiply(
    A=ScalarParameter() {defaultValue=6000, parameterName=Speed}
    B=CollectionParameter() {collection=/Game/Athena/Prototype/Blueprints/Galileo/Blueprints/MPC_Galileo, parameterName=CameraShakeSpeed}
   )
   B=ViewSize()
  )
  n3 = TextureSample(UVs=Multiply(A=Multiply(A=@n1, B=@n2), B=Constant2Vector() {g=0.0156, r=1})) {samplerType=SAMPLERTYPE_LinearColor, texture=&HighContrastNoise}
  n4 = Subtract(A=ComponentMask(None=@n3.RGB) {g=true, r=true}) {constB=0.217}
  n5 = CollectionParameter() {collection=&MPC_Galileo, parameterName=CameraShake}
  n6 = Multiply(A=Multiply(A=ScalarParameter() {defaultValue=0.0015, parameterName=XY_Displace}, B=@n2), B=@n5)
  n7 = Multiply(A=@n3.B, B=ScalarParameter() {defaultValue=-0.01, parameterName=Z_Displace})
  n8 = MaterialFunctionCall(UVs=Add(A=ScreenPosition().ViewportUV, B=Multiply(A=@n4, B=@n6)), Texture Scale=Add(A=Multiply(A=@n7, B=@n5))) {materialFunction=&ScaleUVsByCenter}
  n9 = CollectionParameter() {collection=&MPC_Galileo, parameterName=CameraWind}
  n10 = TextureSample(UVs=Panner(Coordinate=TextureCoordinate() {uTiling=0.1, vTiling=0.1}) {speedX=1, speedY=0.1}) {samplerType=SAMPLERTYPE_LinearColor, texture=&HighContrastNoise}
  n11 = Multiply(A=@n9, B=@n10.R)
  n12 = MaterialFunctionCall(Random=ScalarParameter() {defaultValue=1, parameterName=Param_1}) {materialFunction=&DitherTemporalAA}
  n13 = Add(A=Multiply(A=Multiply(A=@n11, B=@n12.Result), B=Subtract(A=@n8.UVs) {constB=0.5}), B=@n8.UVs)
  n14 = Multiply(A=@n9, B=@n10.G)
  n15 = MaterialFunctionCall(Random=ScalarParameter() {defaultValue=1.5, parameterName=Param_2}) {materialFunction=&DitherTemporalAA}
  n16 = Add(A=Multiply(A=Multiply(A=@n14, B=@n15.Result), B=Subtract(A=@n8.UVs) {constB=0.5}), B=@n8.UVs)
  LinearInterpolate(
   A=SceneTexture(UVs=@n8.UVs) {sceneTextureId=PPI_PostProcessInput0}.Color
   B=LinearInterpolate(A=SceneTexture(UVs=@n13) {sceneTextureId=PPI_PostProcessInput0}.Color, B=SceneTexture(UVs=@n16) {sceneTextureId=PPI_PostProcessInput0}.Color)
   Alpha=CollectionParameter() {collection=&MPC_Galileo, parameterName=PostProcessBlend}
  )
  VertexInterpolator(VS=@n8.UVs)
 ]
}
